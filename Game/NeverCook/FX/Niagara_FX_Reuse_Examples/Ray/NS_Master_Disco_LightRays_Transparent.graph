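# Niagara example system that, by its name, renders transparent disco light rays as mesh particles.
# A single emitter, named Empty, burst-spawns gold, half-alpha meshes with doubled scale and a written material random value.
# Particles are not killed when their lifetime elapses.
# Particle update drives the mesh material through dynamic parameters, one random in the zero-to-one range and the others constant, with no user parameters exposed.
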NiagaraSystem /Game/NeverCook/FX/Niagara_FX_Reuse_Examples/Ray/NS_Master_Disco_LightRays_Transparent
dependencies {usedDataInterfaces [], usedDynamicInputs [], usedModules [&SystemState]}
emitters {Empty {
  emitterUpdateScript [
   &EmitterState(Life Cycle Mode=::NewEnumerator1, Loop Behavior=::NewEnumerator1, Loop Duration Mode=::NewEnumerator1, UseLoopDelay=false)
   &SpawnBurst_Instantaneous(Use Spawn Probability=false, UseLoopCountLimit=false)
  ]
  particleSpawnScript [&InitializeParticle(
    Color=<rgba 0.5 0.3396 0.0315 0.5>
    Mesh Scale=&Multiply_VectorByFloat(Vector=<1 1 1>, Float=2)
    Mesh Scale Mode=::NewEnumerator4
    UsePositionOffset=false
    Write Material Random=true
   )]
  particleUpdateScript [
   &ParticleState(Kill Particles When Lifetime Has Elapsed=false)
   &DynamicMaterialParameters(
    Index 0 Param 1=&RandomRangeFloat(Minimum=0, Maximum=1)
    Index 0 Param 2=0.2
    Index 0 Param 3=1
    Index 0 Param 4=1
    Module.Index0Float0Write=true
    Module.Index0Float1Write=true
    Module.Index0Float2Write=true
    Module.Index0Float3Write=true
   )
  ]
  rendererClasses [&NiagaraMeshRendererProperties]
  renderers [{rendererClass=&NiagaraMeshRendererProperties, rendererIndex=0}]
 }}
summary {
 systemName=NS_Master_Disco_LightRays_Transparent
 emitters [{emitterName=Empty, rendererClasses [&NiagaraMeshRendererProperties]}]
 userVariables []
}
